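# Niagara system for a big spark burst effect in the Sparks stage show.
# It combines spark mesh, two spark spline, glow and flare emitters, whose modules and renderers are not detailed here, with a light emitter.
# The light spawns as an instantaneous burst of bright orange, raised above the origin, and a float curve ramps its color up, holds, then fades it out.
# The system exposes no user parameters.
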NiagaraSystem /SparksMaps/Effects/StageShow/Sparks/Particles/NS_LWE_Sparks_SparkMesh_Burst_Big
dependencies {usedDataInterfaces [&NiagaraDataInterfaceCurve], usedDynamicInputs [], usedModules [&SystemState]}
emitters {
 LWE_Flare {rendererClasses [], renderers []}
 LWE_Glow {rendererClasses [], renderers []}
 LWE_Spark_Mesh {rendererClasses [], renderers []}
 LWE_Spark_Splines {rendererClasses [], renderers []}
 LWE_Spark_Splines_02 {rendererClasses [], renderers []}
 Light001 {
  emitterUpdateScript [
   &EmitterState(Life Cycle Mode=::NewEnumerator1, Loop Behavior=::NewEnumerator1, Loop Duration=1, UseLoopDelay=false)
   &SpawnBurst_Instantaneous(Use Spawn Probability=false, UseLoopCountLimit=false)
  ]
  particleSpawnScript [&InitializeParticle(Color=<rgba 50 9.1435 0 1>, Position Offset=<0 0 100>, UsePositionOffset=true, Write Material Random=true)]
  particleUpdateScript [
   &ParticleState()
   &ScaleColor(
    Scale RGB=&VectorFromFloat(Value=&FloatFromCurve(FloatCurve=DataInterface() {propertyValues {Curve {keys [
          {interpMode=RCIM_Cubic, tangentMode=RCTM_User, time=0, value=0}
          {interpMode=RCIM_Cubic, tangentMode=RCTM_User, time=0.22, value=1}
          {interpMode=RCIM_Cubic, tangentMode=RCTM_User, time=0.8, value=1}
          {interpMode=RCIM_Cubic, tangentMode=RCTM_User, time=1, value=0}
         ]}}}))
    ScaleA=true
    ScaleRGB=true
   )
  ]
  rendererClasses [&NiagaraLightRendererProperties]
  renderers [{rendererClass=&NiagaraLightRendererProperties, rendererIndex=0}]
 }
}
summary {
 systemName=NS_LWE_Sparks_SparkMesh_Burst_Big
 emitters [
  {emitterName=LWE_Spark_Mesh, rendererClasses []}
  {emitterName=LWE_Spark_Splines, rendererClasses []}
  {emitterName=LWE_Glow, rendererClasses []}
  {emitterName=LWE_Flare, rendererClasses []}
  {emitterName=Light001, rendererClasses [&NiagaraLightRendererProperties]}
  {emitterName=LWE_Spark_Splines_02, rendererClasses []}
 ]
 userVariables []
}
